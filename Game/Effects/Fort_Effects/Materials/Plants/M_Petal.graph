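# Masked subsurface material for flower petal particles, usable on sprites and Niagara mesh particles.
# Bump-offset UVs, warped by squaring U, drive a sphere-normal texture and a sphere mask shaping subsurface opacity and color blending.
# Fanned UVs sample a cartoon flame texture to streak the petal gradient, while noise varies roughness.
# Particle random or dynamic parameters vary both textures, and mobile skips bump offset, flame streaks and particle random.
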Material /Game/Effects/Fort_Effects/Materials/Plants/M_Petal
graph {
 properties {
  bUsedWithNiagaraMeshParticles=true
  bUsedWithNiagaraSprites=true
  bUsedWithParticleSprites=true
  blendMode=BLEND_Masked
  opacityMaskClipValue=0.1
  pixelDepthOffsetMode=PDOM_Legacy
  shadingModel=MSM_Subsurface
  usedShadingModels=Subsurface
 }
 nodes [
  n1 = TextureCoordinate()
  n2 = StaticSwitchParameter(
   True=ShadingPathSwitch(Default=ParticleRandom(), Mobile=Constant()) {desc=Migrated from FeatureLevelSwitch}
   False=DynamicParameter() {paramNames [Random,Param2,Param3,Param4]}.Random
  ) {parameterName=Use Particle Random}
  LinearInterpolate(
   A=TextureSample(UVs=Add(A=@n1, B=@n2)) {texture=/Game/Effects/Fort_Effects/Textures/Misc/T_Shot0030_NoiseCutout}.G
   B=Constant() {r=1}
   Alpha=ScalarParameter() {parameterName=Roughness}
  )
  n3 = TextureCoordinate()
  n4 = Constant() {r=0.75}
  n5 = Constant() {r=0.5}
  n6 = SphereMask(A=TextureCoordinate(), B=@n5, Radius=@n5, Hardness=Constant())
  n7 = MaterialFunctionCall(Float2=ShadingPathSwitch(Default=BumpOffset(Coordinate=@n3, Height=@n4, HeightRatioInput=@n6), Mobile=@n3) {desc=Migrated from FeatureLevelSwitch}) {materialFunction=&BreakOutFloat2Components}
  n8 = AppendVector(A=Multiply(A=@n7.R, B=@n7.R), B=@n7.G)
  n9 = Constant() {r=0.5}
  n10 = SphereMask(A=@n8, B=@n9, Radius=@n9, Hardness=Constant())
  Multiply(A=OneMinus(None=@n10), B=ScalarParameter() {defaultValue=0.5, parameterName=SSS_Opacity})
  Multiply(A=Subtract(A=TextureSample(UVs=@n8) {samplerType=SAMPLERTYPE_LinearColor, texture=/Game/Packages/Fortress_SharedMaterials/Textures/SphereNormalsUncomp}.RGB) {constB=0.5}) {constB=2}
  n11 = VectorParameter() {parameterName=Param_1, defaultValue=<rgba 1 0.6533 0.0645 1>}
  n12 = VectorParameter() {defaultValue=<rgba 1 0.0809 0 1>}
  n13 = ComponentMask(None=@n8) {g=true}
  n14 = ComponentMask(None=Multiply(A=@n8) {constB=0.5}) {r=true}
  n15 = ParticleRandom()
  n16 = Constant()
  n17 = AppendVector(
   A=Divide(A=Add(A=@n13) {constB=-0.5}, B=Sine(None=@n14))
   B=StaticSwitchParameter(
    True=ShadingPathSwitch(Default=@n15, Mobile=@n16) {desc=Migrated from FeatureLevelSwitch}
    False=DynamicParameter() {paramNames [Random,Param2,Param3,Param4]}.Random
   ) {parameterName=Use Particle Random}
  )
  n18 = ShadingPathSwitch(
   Default=Multiply(A=@n7.R, B=TextureSample(UVs=@n17) {texture=/Game/Effects/Fort_Effects/Textures/Tile/T_Tile_Texture_Cartoon_Flames_Layer}.G)
   Mobile=@n7.R
  ) {desc=Migrated from FeatureLevelSwitch}
  n19 = Multiply(A=@n10) {constB=3}
  Multiply(
   A=LinearInterpolate(
    A=VectorParameter() {parameterName=Param_2, defaultValue=<rgba 1 0.2303 0.0804 1>}.RGB
    B=LinearInterpolate(A=@n11.RGB, B=@n12.RGB, Alpha=@n18)
    Alpha=Saturate(None=@n19)
   )
   B=ScalarParameter() {defaultValue=0.5, parameterName=SSS_Color}
  )
 ]
}
